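# Control Rig for the M_LRG_SonnetSpirit player body that adds procedural corrective deformation to its skeleton.
# The forward solve scales sleeve bones from wrist rotation, runs the shared DeformRig function and aim-constrains upper-arm helper bones at the forearms.
# Bendy Limbs then offsets knee and elbow curve bones using DeformRig alpha attributes.
# A spherical pose reader on the right upper arm, gated by "is holding gun", shifts chest and beard movers.
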ControlRigBlueprint /BRCosmetics/Characters/Player/Male/Large/Bodies/M_LRG_SonnetSpirit/Meshes/M_LRG_SonnetSpirit_CtrlRig1
graphs [
 New Function_ContainedGraph {nodes [
   RigVMFunctionEntryNode()
   > &RigUnit_ModifyTransforms(
    ItemToModify="((Item=(Type=Bone,Name=\"crv_calf_l_top\"),Transform=(Rotation=(X=0,Y=0,Z=-0.4617,W=0.887),Translation=(X=0,Y=5,Z=0),Scale3D=(X=1,Y=1.5,Z=1))),(Item=(Type=Bone,Name=\"crv_calf_l_bottom\"),Transform=(Rotation=(X=0,Y=0,Z=-0.1718,W=0.9851),Translation=(X=0,Y=0,Z= ...(truncated)"
    WeightMinimum=0
    Weight=RigVMDispatchNode(BoneName=root, Default=0, Name=deformRig_l_knee_alpha) {template="DISPATCH_RigDispatch_GetAnimAttribute(in Name,in BoneName,in Default,out Value,out Found)"}.Value
   )
   > &RigUnit_ModifyTransforms(
    ItemToModify="((Item=(Type=Bone,Name=\"crv_calf_r_top\"),Transform=(Rotation=(X=0,Y=0,Z=-0.4617,W=0.887),Translation=(X=1.239,Y=2.5034,Z=0),Scale3D=(X=1,Y=1,Z=1))),(Item=(Type=Bone,Name=\"crv_calf_r_bottom\"),Transform=(Rotation=(X=0,Y=0,Z=-0.1715,W=0.9852),Translation=(X=0 ...(truncated)"
    WeightMinimum=0
    Weight=RigVMDispatchNode(BoneName=root, Default=0, Name=deformRig_r_knee_alpha) {template="DISPATCH_RigDispatch_GetAnimAttribute(in Name,in BoneName,in Default,out Value,out Found)"}.Value
   )
   > &RigUnit_ModifyTransforms(
    ItemToModify="((Item=(Type=Bone,Name=\"crv_lowerarm_l_top\"),Transform=(Rotation=(X=0,Y=0,Z=-0.5,W=0.866),Translation=(X=-10,Y=0,Z=0),Scale3D=(X=1,Y=1,Z=1))),(Item=(Type=Bone,Name=\"crv_lowerarm_l_bottom\"),Transform=(Rotation=(X=0,Y=0,Z=0,W=1),Translation=(X=0,Y=-10,Z=0),S ...(truncated)"
    WeightMinimum=0
    Weight=RigVMDispatchNode(BoneName=root, Default=0, Name=deformRig_l_elbow_alpha) {template="DISPATCH_RigDispatch_GetAnimAttribute(in Name,in BoneName,in Default,out Value,out Found)"}.Value
   )
   > &RigUnit_ModifyTransforms(
    ItemToModify="((Item=(Type=Bone,Name=\"crv_lowerarm_r_top\"),Transform=(Rotation=(X=0,Y=0,Z=-0.5,W=0.866),Translation=(X=10,Y=0,Z=0),Scale3D=(X=1,Y=1,Z=1))),(Item=(Type=Bone,Name=\"crv_lowerarm_r_bottom\"),Transform=(Rotation=(X=0,Y=0,Z=0,W=1),Translation=(X=0,Y=10,Z=0),Sca ...(truncated)"
    WeightMinimum=0
    Weight=RigVMDispatchNode(BoneName=root, Default=0, Name=deformRig_r_elbow_alpha) {template="DISPATCH_RigDispatch_GetAnimAttribute(in Name,in BoneName,in Default,out Value,out Found)"}.Value
   )
   > RigVMFunctionReturnNode()
  ]}
 RigVMFunctionLibrary {nodes [Bendy Limbs = RigVMCollapseNode(ExecuteContext=())]}
 RigVMModel {nodes [
   RigUnit_BeginExecution = &RigUnit_BeginExecution()
   n1 = &RigUnit_GetRelativeTransformForItem(Child=(Type=Bone,Name="hand_l"), Parent=(Type=Bone,Name="lowerarm_l"), bChildInitial=False, bParentInitial=False)
   &RigUnit_ModifyTransforms(
    ItemToModify=((Item=(Type=Bone,Name="dyn_L_sleeveScale"),Transform=(Rotation=(X=0,Y=0,Z=0,W=1),Translation=(X=0,Y=0,Z=0),Scale3D=(X=0.4,Y=1,Z=1))))
    WeightMinimum=0
    Weight=&RigVMFunction_MathDoubleRemap(
     SourceMaximum=-60
     SourceMinimum=-10
     TargetMaximum=1
     TargetMinimum=0
     Value=&RigVMFunction_MathQuaternionToEuler(RotationOrder=ZYX, Value=@n1.RelativeTransform.Rotation).Result.Z
    ).Result
    ExecutePin=@RigUnit_BeginExecution.ExecutePin
   )
   n2 = &RigUnit_GetRelativeTransformForItem(Child=(Type=Bone,Name="hand_r"), Parent=(Type=Bone,Name="lowerarm_r"), bChildInitial=False, bParentInitial=False)
   > &RigUnit_ModifyTransforms(
    ItemToModify=((Item=(Type=Bone,Name="dyn_R_sleeveScale"),Transform=(Rotation=(X=0,Y=0,Z=0,W=1),Translation=(X=0,Y=0,Z=0),Scale3D=(X=0.4,Y=1,Z=1))))
    WeightMinimum=0
    Weight=&RigVMFunction_MathDoubleRemap(
     SourceMaximum=-60
     SourceMinimum=-10
     TargetMaximum=1
     TargetMinimum=0
     Value=&RigVMFunction_MathQuaternionToEuler(RotationOrder=ZYX, Value=@n2.RelativeTransform.Rotation).Result.Z
    ).Result
   )
   > FN_DeformRig_FNC = RigVMFunctionReferenceNode(
    DeformRig_ Curve="(deformRig_l_upperarm_up_alpha_72_#1=False,deformRig_r_upperarm_up_alpha_73_#2=true,deformRig_l_upperarm_fr_alpha_74_#3=true,deformRig_r_upperarm_fr_alpha_75_CB51B80C ...(truncated)"
    DeformRig_Bool="(upperarm_1_#4=True,elbow_3_#5=True,wrist_7_#6=True,groin_8_#7=True,glute_17_#8=True,knee_9_40C408324CA261 ...(truncated)"
    DeformRig_Debug=(upperarm_debug_18_#4=False,wrist_debug_20_#6=False,lat_debug_26_#9=False,groinGlute_debug_27_#7=False)
    DeformRig_Value="(upperarm_up_18_#10=(Rotation=(X=0.0,Y=-0.0,Z=0.0,W=1.0),Translation=(X=0.0,Y=0.0,Z=8.0),Scale3D=(X=0.70,Y=1.0,Z=1.0)),upperarm_fwd_fr_19_#11=(Rotation=(X=0.0,Y=-0.0,Z=0.0,W=1.0),Translation=(X=0.0, ...(truncated)"
   )
   > &RigUnit_ParentConstraint(
    AdvancedSettings=(InterpolationType=Average,RotationOrderForFilter=XZY)
    Child=(Type=Bone,Name="aim_upperarm_l")
    Filter=(TranslationFilter=(bX=True,bY=True,bZ=True),RotationFilter=(bX=false,bY=false,bZ=false),ScaleFilter=(bX=false,bY=false,bZ=false))
    Parents=((Item=(Type=Bone,Name="clavicle_l"),Weight=1))
    Weight=1
    bMaintainOffset=true
   )
   > &RigUnit_AimConstraintLocalSpaceOffset(
    AdvancedSettings=(DebugSettings=(bEnabled=false,Scale=10,WorldOffset=(Rotation=(X=0,Y=0,Z=0,W=1),Translation=(X=0,Y=0,Z=0),Scale3D=(X=1,Y=1,Z=1))),RotationOrderForFilter=XZY)
    Child=(Type=Bone,Name="aim_upperarm_l")
    Filter=(bX=True,bY=True,bZ=True)
    Parents=((Item=(Type=Bone,Name="lowerarm_l"),Weight=1))
    Weight=1
    WorldUp=(Target=(X=0,Y=0,Z=1),Kind=Direction,Space=(Type=Bone,Name="lowerarm_l"))
    bMaintainOffset=true
   )
   > &RigUnit_ParentConstraint(
    AdvancedSettings=(InterpolationType=Average,RotationOrderForFilter=XZY)
    Child=(Type=Bone,Name="aim_upperarm_r")
    Filter=(TranslationFilter=(bX=True,bY=True,bZ=True),RotationFilter=(bX=false,bY=false,bZ=false),ScaleFilter=(bX=false,bY=false,bZ=false))
    Parents=((Item=(Type=Bone,Name="clavicle_r"),Weight=1))
    Weight=1
    bMaintainOffset=true
   )
   > &RigUnit_AimConstraintLocalSpaceOffset(
    AdvancedSettings=(DebugSettings=(bEnabled=false,Scale=10,WorldOffset=(Rotation=(X=0,Y=0,Z=0,W=1),Translation=(X=0,Y=0,Z=0),Scale3D=(X=1,Y=1,Z=1))),RotationOrderForFilter=XYZ)
    Child=(Type=Bone,Name="aim_upperarm_r")
    Filter=(bX=true,bY=True,bZ=True)
    Parents=((Item=(Type=Bone,Name="lowerarm_r"),Weight=1))
    Weight=1
    WorldUp=(Target=(X=0,Y=0,Z=-1),Kind=Direction,Space=(Type=Bone,Name="lowerarm_r"))
    bMaintainOffset=true
   )
   > Bendy Limbs_1 = RigVMFunctionReferenceNode()
   > n3 = &RigUnit_SphericalPoseReader(
    ActiveRegionScaleFactors=(PositiveWidth=1,NegativeWidth=1,PositiveHeight=1,NegativeHeight=1)
    ActiveRegionSize=0.1
    Debug=(bDrawDebug=false,bDraw2D=False,bDrawLocalAxes=False,DebugScale=25,DebugSegments=20,DebugThickness=0.25)
    DriverAxis=(X=-1,Y=0,Z=0)
    DriverItem=(Type=Bone,Name="aim_upperarm_r")
    FalloffRegionScaleFactors=(PositiveWidth=1,NegativeWidth=1,PositiveHeight=1,NegativeHeight=1)
    FalloffSize=0.3
    FlipHeightScaling=False
    FlipWidthScaling=False
    OptionalParentItem=(Type=Bone,Name="spine_05")
    RotationOffset=(X=150,Y=0,Z=-30)
   )
   Multiply = &RigVMFunction_MathDoubleMul(A=@n3.OutputParam, B=RigVMVariableNode() {variable=is holding gun}.Value)
   > &RigUnit_ModifyTransforms(
    ItemToModify="((Item=(Type=Bone,Name=\"Chest_mover_r\"),Transform=(Rotation=(X=-0.0002,Y=0,Z=0,W=-1),Translation=(X=0,Y=-10.3903,Z=1.3979),Scale3D=(X=1,Y=1,Z=1))),(Item=(Type=Bone,Name=\"Beared_Mover_L\"),Transform=(Rotation=(X=0,Y=0,Z=0,W=1),Translation=(X=0,Y=-5,Z=0),Scal ...(truncated)"
    WeightMinimum=0
    Weight=@Multiply.Result
   )
  ]}
 "RigVMModel Contruction " {nodes [
   n1 = &RigUnit_PrepareForExecution()
   &RigUnit_HierarchyImportFromSkeleton(NameSpace=None, bIncludeCurves=False, bIncludeMeshSockets=False, bIncludeVirtualBones=True, ExecutePin=@n1.ExecutePin)
   > addFKCtrl = RigVMFunctionReferenceNode(
    Custom_FKCtrl=()
    Options_FKCtrl="(pelvis_1_#12=false,spine_3_#13=false,clavicle_5_#14=false,arm_7_#15=false,leg_9_#16=false,head_11_46DC15EF465C ...(truncated)"
    Shape_Settings="(pelvis_2_#17=(bVisible=True,Name=\"Hexagon_Thin\",Color=(R=0.0117,G=0.8281,B=1,A=1),Transform=(Rotation=(X=0,Y=-0.7071,Z=0,W=0.7071),Translation=(X=2,Y=0,Z=0),Scale3D=(X=5,Y=5,Z=7))),spine_4_#18=(bVi ...(truncated)"
   )
  ]}
]
hierarchy [
 look_left_pose {type=Curve, loc=<0 0 0>, rot=<0 0 0 1>}
 look_right_pose {type=Curve, loc=<0 0 0>, rot=<0 0 0 1>}
 look_up_pose {type=Curve, loc=<0 0 0>, rot=<0 0 0 1>}
 look_down_pose {type=Curve, loc=<0 0 0>, rot=<0 0 0 1>}
 L_blink_pose {type=Curve, loc=<0 0 0>, rot=<0 0 0 1>}
 R_blink_pose {type=Curve, loc=<0 0 0>, rot=<0 0 0 1>}
 L_wide_pose {type=Curve, loc=<0 0 0>, rot=<0 0 0 1>}
 R_wide_pose {type=Curve, loc=<0 0 0>, rot=<0 0 0 1>}
 R_squint_inner_pose {type=Curve, loc=<0 0 0>, rot=<0 0 0 1>}
 R_squeeze_pose {type=Curve, loc=<0 0 0>, rot=<0 0 0 1>}
 L_squint_inner_pose {type=Curve, loc=<0 0 0>, rot=<0 0 0 1>}
 L_squeeze_pose {type=Curve, loc=<0 0 0>, rot=<0 0 0 1>}
 L_brow_up_pose {type=Curve, loc=<0 0 0>, rot=<0 0 0 1>}
 L_brow_down_pose {type=Curve, loc=<0 0 0>, rot=<0 0 0 1>}
 R_brow_up_pose {type=Curve, loc=<0 0 0>, rot=<0 0 0 1>}
 R_brow_down_pose {type=Curve, loc=<0 0 0>, rot=<0 0 0 1>}
 C_glabella_up_pose {type=Curve, loc=<0 0 0>, rot=<0 0 0 1>}
 C_glabella_down_pose {type=Curve, loc=<0 0 0>, rot=<0 0 0 1>}
 L_brow_squeeze_in_pose {type=Curve, loc=<0 0 0>, rot=<0 0 0 1>}
 R_brow_squeeze_in_pose {type=Curve, loc=<0 0 0>, rot=<0 0 0 1>}
 L_upper_lip_raiser_pose {type=Curve, loc=<0 0 0>, rot=<0 0 0 1>}
 R_upper_lip_raiser_pose {type=Curve, loc=<0 0 0>, rot=<0 0 0 1>}
 L_upper_lip_lower_pose {type=Curve, loc=<0 0 0>, rot=<0 0 0 1>}
 R_upper_lip_lower_pose {type=Curve, loc=<0 0 0>, rot=<0 0 0 1>}
 L_lower_lip_up_pose {type=Curve, loc=<0 0 0>, rot=<0 0 0 1>}
 R_lower_lip_up_pose {type=Curve, loc=<0 0 0>, rot=<0 0 0 1>}
 L_lower_lip_down_pose {type=Curve, loc=<0 0 0>, rot=<0 0 0 1>}
 R_lower_lip_down_pose {type=Curve, loc=<0 0 0>, rot=<0 0 0 1>}
 L_lip_corner_narrow_pose {type=Curve, loc=<0 0 0>, rot=<0 0 0 1>}
 L_lip_corner_wide_pose {type=Curve, loc=<0 0 0>, rot=<0 0 0 1>}
 R_lip_corner_narrow_pose {type=Curve, loc=<0 0 0>, rot=<0 0 0 1>}
 R_lip_corner_wide_pose {type=Curve, loc=<0 0 0>, rot=<0 0 0 1>}
 L_smile_pose {type=Curve, loc=<0 0 0>, rot=<0 0 0 1>}
 R_smile_pose {type=Curve, loc=<0 0 0>, rot=<0 0 0 1>}
 L_frown_pose {type=Curve, loc=<0 0 0>, rot=<0 0 0 1>}
 R_frown_pose {type=Curve, loc=<0 0 0>, rot=<0 0 0 1>}
 phoneme_oo_pose {type=Curve, loc=<0 0 0>, rot=<0 0 0 1>}
 phoneme_ch_pose {type=Curve, loc=<0 0 0>, rot=<0 0 0 1>}
 phoneme_mbp_pose {type=Curve, loc=<0 0 0>, rot=<0 0 0 1>}
 phoneme_fv_pose {type=Curve, loc=<0 0 0>, rot=<0 0 0 1>}
 mouth_up_pose {type=Curve, loc=<0 0 0>, rot=<0 0 0 1>}
 mouth_down_pose {type=Curve, loc=<0 0 0>, rot=<0 0 0 1>}
 C_nose_wrinkler_pose {type=Curve, loc=<0 0 0>, rot=<0 0 0 1>}
 L_cheek_up_pose {type=Curve, loc=<0 0 0>, rot=<0 0 0 1>}
 R_cheek_up_pose {type=Curve, loc=<0 0 0>, rot=<0 0 0 1>}
 jaw_open_pose {type=Curve, loc=<0 0 0>, rot=<0 0 0 1>}
 jaw_right_pose {type=Curve, loc=<0 0 0>, rot=<0 0 0 1>}
 jaw_left_pose {type=Curve, loc=<0 0 0>, rot=<0 0 0 1>}
 eye_convergence_pose {type=Curve, loc=<0 0 0>, rot=<0 0 0 1>}
 L_upperLidDown_pose {type=Curve, loc=<0 0 0>, rot=<0 0 0 1>}
 R_upperLidDown_pose {type=Curve, loc=<0 0 0>, rot=<0 0 0 1>}
 jaw_up_pose {type=Curve, loc=<0 0 0>, rot=<0 0 0 1>}
 jaw_forward_pose {type=Curve, loc=<0 0 0>, rot=<0 0 0 1>}
 jaw_back_pose {type=Curve, loc=<0 0 0>, rot=<0 0 0 1>}
 tongue_forward_pose {type=Curve, loc=<0 0 0>, rot=<0 0 0 1>}
 tongue_back_pose {type=Curve, loc=<0 0 0>, rot=<0 0 0 1>}
 tongue_up_pose {type=Curve, loc=<0 0 0>, rot=<0 0 0 1>}
 tongue_down_pose {type=Curve, loc=<0 0 0>, rot=<0 0 0 1>}
 base_pose {type=Curve, loc=<0 0 0>, rot=<0 0 0 1>}
 Hair_Scarf {type=Curve, loc=<0 0 0>, rot=<0 0 0 1>}
 Hair_Hat {type=Curve, loc=<0 0 0>, rot=<0 0 0 1>}
 Hair_Helmet {type=Curve, loc=<0 0 0>, rot=<0 0 0 1>}
 Hair_Glasses {type=Curve, loc=<0 0 0>, rot=<0 0 0 1>}
 Hair_Cap {type=Curve, loc=<0 0 0>, rot=<0 0 0 1>}
 Hair_Headsets {type=Curve, loc=<0 0 0>, rot=<0 0 0 1>}
 Hair_Hood {type=Curve, loc=<0 0 0>, rot=<0 0 0 1>}
 Face_Mask {type=Curve, loc=<0 0 0>, rot=<0 0 0 1>}
 HairStrands_Hat {type=Curve, loc=<0 0 0>, rot=<0 0 0 1>}
 Hat_Hair {type=Curve, loc=<0 0 0>, rot=<0 0 0 1>}
 Helmet_Hair {type=Curve, loc=<0 0 0>, rot=<0 0 0 1>}
 Hair_Muff {type=Curve, loc=<0 0 0>, rot=<0 0 0 1>}
 Hair_SpaceSuit {type=Curve, loc=<0 0 0>, rot=<0 0 0 1>}
 Hair_Hide {type=Curve, loc=<0 0 0>, rot=<0 0 0 1>}
 Head_SpaceSuit {type=Curve, loc=<0 0 0>, rot=<0 0 0 1>}
 Hair_Null {type=Curve, loc=<0 0 0>, rot=<0 0 0 1>}
 DisableFaceOverride {type=Curve, loc=<0 0 0>, rot=<0 0 0 1>}
 LiveLink {type=Curve, loc=<0 0 0>, rot=<0 0 0 1>}
 Hat_MESH {type=Curve, loc=<0 0 0>, rot=<0 0 0 1>}
 Null_MESH {type=Curve, loc=<0 0 0>, rot=<0 0 0 1>}
 Pose_0 {type=Curve, loc=<0 0 0>, rot=<0 0 0 1>}
 Hair_Mask {type=Curve, loc=<0 0 0>, rot=<0 0 0 1>}
 eye_convergence {type=Curve, loc=<0 0 0>, rot=<0 0 0 1>}
 headTransform {type=Curve, loc=<0 0 0>, rot=<0 0 0 1>}
 Pose_1 {type=Curve, loc=<0 0 0>, rot=<0 0 0 1>}
 Pose_2 {type=Curve, loc=<0 0 0>, rot=<0 0 0 1>}
 Pose_3 {type=Curve, loc=<0 0 0>, rot=<0 0 0 1>}
 Pose_4 {type=Curve, loc=<0 0 0>, rot=<0 0 0 1>}
 Pose_5 {type=Curve, loc=<0 0 0>, rot=<0 0 0 1>}
 Pose_6 {type=Curve, loc=<0 0 0>, rot=<0 0 0 1>}
 Pose_7 {type=Curve, loc=<0 0 0>, rot=<0 0 0 1>}
 Pose_8 {type=Curve, loc=<0 0 0>, rot=<0 0 0 1>}
 Pose_9 {type=Curve, loc=<0 0 0>, rot=<0 0 0 1>}
 Pose_10 {type=Curve, loc=<0 0 0>, rot=<0 0 0 1>}
 Pose_11 {type=Curve, loc=<0 0 0>, rot=<0 0 0 1>}
 Pose_12 {type=Curve, loc=<0 0 0>, rot=<0 0 0 1>}
 Pose_13 {type=Curve, loc=<0 0 0>, rot=<0 0 0 1>}
 Pose_14 {type=Curve, loc=<0 0 0>, rot=<0 0 0 1>}
 Pose_15 {type=Curve, loc=<0 0 0>, rot=<0 0 0 1>}
 Pose_16 {type=Curve, loc=<0 0 0>, rot=<0 0 0 1>}
 Pose_17 {type=Curve, loc=<0 0 0>, rot=<0 0 0 1>}
 Pose_18 {type=Curve, loc=<0 0 0>, rot=<0 0 0 1>}
 Pose_19 {type=Curve, loc=<0 0 0>, rot=<0 0 0 1>}
 Pose_20 {type=Curve, loc=<0 0 0>, rot=<0 0 0 1>}
 Pose_21 {type=Curve, loc=<0 0 0>, rot=<0 0 0 1>}
 Pose_22 {type=Curve, loc=<0 0 0>, rot=<0 0 0 1>}
 Pose_23 {type=Curve, loc=<0 0 0>, rot=<0 0 0 1>}
 Pose_24 {type=Curve, loc=<0 0 0>, rot=<0 0 0 1>}
 Pose_25 {type=Curve, loc=<0 0 0>, rot=<0 0 0 1>}
 Pose_26 {type=Curve, loc=<0 0 0>, rot=<0 0 0 1>}
 Pose_27 {type=Curve, loc=<0 0 0>, rot=<0 0 0 1>}
 Pose_28 {type=Curve, loc=<0 0 0>, rot=<0 0 0 1>}
 Pose_29 {type=Curve, loc=<0 0 0>, rot=<0 0 0 1>}
 Pose_30 {type=Curve, loc=<0 0 0>, rot=<0 0 0 1>}
 Pose_31 {type=Curve, loc=<0 0 0>, rot=<0 0 0 1>}
 Pose_32 {type=Curve, loc=<0 0 0>, rot=<0 0 0 1>}
 Pose_33 {type=Curve, loc=<0 0 0>, rot=<0 0 0 1>}
 Pose_34 {type=Curve, loc=<0 0 0>, rot=<0 0 0 1>}
 Pose_35 {type=Curve, loc=<0 0 0>, rot=<0 0 0 1>}
 Pose_36 {type=Curve, loc=<0 0 0>, rot=<0 0 0 1>}
 Pose_37 {type=Curve, loc=<0 0 0>, rot=<0 0 0 1>}
 Pose_38 {type=Curve, loc=<0 0 0>, rot=<0 0 0 1>}
 Pose_39 {type=Curve, loc=<0 0 0>, rot=<0 0 0 1>}
 Pose_40 {type=Curve, loc=<0 0 0>, rot=<0 0 0 1>}
 Pose_41 {type=Curve, loc=<0 0 0>, rot=<0 0 0 1>}
 Pose_42 {type=Curve, loc=<0 0 0>, rot=<0 0 0 1>}
 Pose_43 {type=Curve, loc=<0 0 0>, rot=<0 0 0 1>}
 Pose_44 {type=Curve, loc=<0 0 0>, rot=<0 0 0 1>}
 Pose_45 {type=Curve, loc=<0 0 0>, rot=<0 0 0 1>}
 Pose_46 {type=Curve, loc=<0 0 0>, rot=<0 0 0 1>}
 Pose_47 {type=Curve, loc=<0 0 0>, rot=<0 0 0 1>}
 Pose_48 {type=Curve, loc=<0 0 0>, rot=<0 0 0 1>}
 Pose_49 {type=Curve, loc=<0 0 0>, rot=<0 0 0 1>}
 Pose_50 {type=Curve, loc=<0 0 0>, rot=<0 0 0 1>}
 Pose_51 {type=Curve, loc=<0 0 0>, rot=<0 0 0 1>}
 Pose_52 {type=Curve, loc=<0 0 0>, rot=<0 0 0 1>}
 Pose_53 {type=Curve, loc=<0 0 0>, rot=<0 0 0 1>}
 Pose_54 {type=Curve, loc=<0 0 0>, rot=<0 0 0 1>}
 Pose_55 {type=Curve, loc=<0 0 0>, rot=<0 0 0 1>}
 Pose_56 {type=Curve, loc=<0 0 0>, rot=<0 0 0 1>}
 Pose_57 {type=Curve, loc=<0 0 0>, rot=<0 0 0 1>}
 Pose_58 {type=Curve, loc=<0 0 0>, rot=<0 0 0 1>}
 Pose_59 {type=Curve, loc=<0 0 0>, rot=<0 0 0 1>}
 Pose_60 {type=Curve, loc=<0 0 0>, rot=<0 0 0 1>}
 Pose_61 {type=Curve, loc=<0 0 0>, rot=<0 0 0 1>}
 Pose_62 {type=Curve, loc=<0 0 0>, rot=<0 0 0 1>}
 Pose_63 {type=Curve, loc=<0 0 0>, rot=<0 0 0 1>}
 Pose_64 {type=Curve, loc=<0 0 0>, rot=<0 0 0 1>}
 Pose_65 {type=Curve, loc=<0 0 0>, rot=<0 0 0 1>}
 Pose_66 {type=Curve, loc=<0 0 0>, rot=<0 0 0 1>}
 Pose_67 {type=Curve, loc=<0 0 0>, rot=<0 0 0 1>}
 Pose_68 {type=Curve, loc=<0 0 0>, rot=<0 0 0 1>}
 Pose_69 {type=Curve, loc=<0 0 0>, rot=<0 0 0 1>}
 eyeLookDownRight {type=Curve, loc=<0 0 0>, rot=<0 0 0 1>}
 eyeLookOutRight {type=Curve, loc=<0 0 0>, rot=<0 0 0 1>}
 eyeLookUpRight {type=Curve, loc=<0 0 0>, rot=<0 0 0 1>}
 eyeLookOutLeft {type=Curve, loc=<0 0 0>, rot=<0 0 0 1>}
 eyeBlinkLeft {type=Curve, loc=<0 0 0>, rot=<0 0 0 1>}
 eyeBlinkRight {type=Curve, loc=<0 0 0>, rot=<0 0 0 1>}
 eyeSquintLeft {type=Curve, loc=<0 0 0>, rot=<0 0 0 1>}
 eyeSquintRight {type=Curve, loc=<0 0 0>, rot=<0 0 0 1>}
 eyeWideLeft {type=Curve, loc=<0 0 0>, rot=<0 0 0 1>}
 eyeWideRight {type=Curve, loc=<0 0 0>, rot=<0 0 0 1>}
 jawForward {type=Curve, loc=<0 0 0>, rot=<0 0 0 1>}
 jawLeft {type=Curve, loc=<0 0 0>, rot=<0 0 0 1>}
 jawRight {type=Curve, loc=<0 0 0>, rot=<0 0 0 1>}
 jawOpen {type=Curve, loc=<0 0 0>, rot=<0 0 0 1>}
 mouthClose {type=Curve, loc=<0 0 0>, rot=<0 0 0 1>}
 mouthFunnel {type=Curve, loc=<0 0 0>, rot=<0 0 0 1>}
 mouthPucker {type=Curve, loc=<0 0 0>, rot=<0 0 0 1>}
 mouthLeft {type=Curve, loc=<0 0 0>, rot=<0 0 0 1>}
 mouthRight {type=Curve, loc=<0 0 0>, rot=<0 0 0 1>}
 mouthSmileLeft {type=Curve, loc=<0 0 0>, rot=<0 0 0 1>}
 mouthSmileRight {type=Curve, loc=<0 0 0>, rot=<0 0 0 1>}
 mouthFrownLeft {type=Curve, loc=<0 0 0>, rot=<0 0 0 1>}
 mouthFrownRight {type=Curve, loc=<0 0 0>, rot=<0 0 0 1>}
 mouthDimpleLeft {type=Curve, loc=<0 0 0>, rot=<0 0 0 1>}
 mouthDimpleRight {type=Curve, loc=<0 0 0>, rot=<0 0 0 1>}
 mouthStretchLeft {type=Curve, loc=<0 0 0>, rot=<0 0 0 1>}
 mouthStretchRight {type=Curve, loc=<0 0 0>, rot=<0 0 0 1>}
 mouthRollLower {type=Curve, loc=<0 0 0>, rot=<0 0 0 1>}
 mouthRollUpper {type=Curve, loc=<0 0 0>, rot=<0 0 0 1>}
 mouthShrugLower {type=Curve, loc=<0 0 0>, rot=<0 0 0 1>}
 mouthShrugUpper {type=Curve, loc=<0 0 0>, rot=<0 0 0 1>}
 mouthPressLeft {type=Curve, loc=<0 0 0>, rot=<0 0 0 1>}
 mouthPressRight {type=Curve, loc=<0 0 0>, rot=<0 0 0 1>}
 mouthLowerDownLeft {type=Curve, loc=<0 0 0>, rot=<0 0 0 1>}
 mouthLowerDownRight {type=Curve, loc=<0 0 0>, rot=<0 0 0 1>}
 mouthUpperUpLeft {type=Curve, loc=<0 0 0>, rot=<0 0 0 1>}
 mouthUpperUpRight {type=Curve, loc=<0 0 0>, rot=<0 0 0 1>}
 browDownLeft {type=Curve, loc=<0 0 0>, rot=<0 0 0 1>}
 browDownRight {type=Curve, loc=<0 0 0>, rot=<0 0 0 1>}
 browInnerUp {type=Curve, loc=<0 0 0>, rot=<0 0 0 1>}
 browOuterUpLeft {type=Curve, loc=<0 0 0>, rot=<0 0 0 1>}
 browOuterUpRight {type=Curve, loc=<0 0 0>, rot=<0 0 0 1>}
 cheekPuff {type=Curve, loc=<0 0 0>, rot=<0 0 0 1>}
 cheekSquintLeft {type=Curve, loc=<0 0 0>, rot=<0 0 0 1>}
 cheekSquintRight {type=Curve, loc=<0 0 0>, rot=<0 0 0 1>}
 noseSneerLeft {type=Curve, loc=<0 0 0>, rot=<0 0 0 1>}
 noseSneerRight {type=Curve, loc=<0 0 0>, rot=<0 0 0 1>}
 C_glabella_down {type=Curve, loc=<0 0 0>, rot=<0 0 0 1>}
 L_brow_squeeze_in {type=Curve, loc=<0 0 0>, rot=<0 0 0 1>}
 R_brow_squeeze_in {type=Curve, loc=<0 0 0>, rot=<0 0 0 1>}
 L_upper_lip_lower {type=Curve, loc=<0 0 0>, rot=<0 0 0 1>}
 R_upper_lip_lower {type=Curve, loc=<0 0 0>, rot=<0 0 0 1>}
 L_lower_lip_up {type=Curve, loc=<0 0 0>, rot=<0 0 0 1>}
 R_lower_lip_up {type=Curve, loc=<0 0 0>, rot=<0 0 0 1>}
 L_lip_corner_narrow {type=Curve, loc=<0 0 0>, rot=<0 0 0 1>}
 R_lip_corner_narrow {type=Curve, loc=<0 0 0>, rot=<0 0 0 1>}
 phoneme_ch {type=Curve, loc=<0 0 0>, rot=<0 0 0 1>}
 phoneme_mbp {type=Curve, loc=<0 0 0>, rot=<0 0 0 1>}
 phoneme_fv {type=Curve, loc=<0 0 0>, rot=<0 0 0 1>}
 mouth_up {type=Curve, loc=<0 0 0>, rot=<0 0 0 1>}
 mouth_down {type=Curve, loc=<0 0 0>, rot=<0 0 0 1>}
 L_cheek_up {type=Curve, loc=<0 0 0>, rot=<0 0 0 1>}
 R_cheek_up {type=Curve, loc=<0 0 0>, rot=<0 0 0 1>}
 L_upperLidDown {type=Curve, loc=<0 0 0>, rot=<0 0 0 1>}
 R_upperLidDown {type=Curve, loc=<0 0 0>, rot=<0 0 0 1>}
 jaw_up {type=Curve, loc=<0 0 0>, rot=<0 0 0 1>}
 jaw_back {type=Curve, loc=<0 0 0>, rot=<0 0 0 1>}
 tongue_forward {type=Curve, loc=<0 0 0>, rot=<0 0 0 1>}
 tongue_back {type=Curve, loc=<0 0 0>, rot=<0 0 0 1>}
 tongue_up {type=Curve, loc=<0 0 0>, rot=<0 0 0 1>}
 tongue_down {type=Curve, loc=<0 0 0>, rot=<0 0 0 1>}
 basePose {type=Curve, loc=<0 0 0>, rot=<0 0 0 1>}
 neutral {type=Curve, loc=<0 0 0>, rot=<0 0 0 1>}
 cheek_puff_low {type=Curve, loc=<0 0 0>, rot=<0 0 0 1>}
 Desert_Island_Jonesy_Mouth {type=Curve, loc=<0 0 0>, rot=<0 0 0 1>}
 HT_L1_TransitionCurve {type=Curve, loc=<0 0 0>, rot=<0 0 0 1>}
 Hair_CapShape {type=Curve, loc=<0 0 0>, rot=<0 0 0 1>}
 Hair_CapShape1 {type=Curve, loc=<0 0 0>, rot=<0 0 0 1>}
 jekyll {type=Curve, loc=<0 0 0>, rot=<0 0 0 1>}
 JH_TransformCurve {type=Curve, loc=<0 0 0>, rot=<0 0 0 1>}
 eyes_LOD0 {type=Curve, loc=<0 0 0>, rot=<0 0 0 1>}
 MaterialCurveA {type=Curve, loc=<0 0 0>, rot=<0 0 0 1>}
 MaterialCurveB {type=Curve, loc=<0 0 0>, rot=<0 0 0 1>}
 MaterialCurveC {type=Curve, loc=<0 0 0>, rot=<0 0 0 1>}
 MaterialCurveD {type=Curve, loc=<0 0 0>, rot=<0 0 0 1>}
 MaterialCurveE {type=Curve, loc=<0 0 0>, rot=<0 0 0 1>}
 Custom_Frontend_Face_Idle {type=Curve, loc=<0 0 0>, rot=<0 0 0 1>}
 blendParent1 {type=Curve, loc=<0 0 0>, rot=<0 0 0 1>}
 BaseHeadDefaultFaceIdle {type=Curve, loc=<0 0 0>, rot=<0 0 0 1>}
 VeiledTransitionGlowAmount {type=Curve, loc=<0 0 0>, rot=<0 0 0 1>}
 is_3L {type=Curve, loc=<0 0 0>, rot=<0 0 0 1>}
 Hair_Hat_LOD0 {type=Curve, loc=<0 0 0>, rot=<0 0 0 1>}
 Hair_Helmet_LOD0 {type=Curve, loc=<0 0 0>, rot=<0 0 0 1>}
 Hair_Null_LOD0 {type=Curve, loc=<0 0 0>, rot=<0 0 0 1>}
 mouth_close_pose {type=Curve, loc=<0 0 0>, rot=<0 0 0 1>}
 mouth_pucker_pose {type=Curve, loc=<0 0 0>, rot=<0 0 0 1>}
 mouth_left_pose {type=Curve, loc=<0 0 0>, rot=<0 0 0 1>}
 mouth_right_pose {type=Curve, loc=<0 0 0>, rot=<0 0 0 1>}
 L_mouth_stretch_pose {type=Curve, loc=<0 0 0>, rot=<0 0 0 1>}
 R_mouth_stretch_pose {type=Curve, loc=<0 0 0>, rot=<0 0 0 1>}
 mouth_roll_lower_pose {type=Curve, loc=<0 0 0>, rot=<0 0 0 1>}
 mouth_roll_upper_pose {type=Curve, loc=<0 0 0>, rot=<0 0 0 1>}
 mouth_shrug_lower_pose {type=Curve, loc=<0 0 0>, rot=<0 0 0 1>}
 mouth_shrug_upper_pose {type=Curve, loc=<0 0 0>, rot=<0 0 0 1>}
 L_mouth_press_pose {type=Curve, loc=<0 0 0>, rot=<0 0 0 1>}
 R_mouth_press_pose {type=Curve, loc=<0 0 0>, rot=<0 0 0 1>}
 cheek_puff_pose {type=Curve, loc=<0 0 0>, rot=<0 0 0 1>}
 L_nose_wrinkler_pose {type=Curve, loc=<0 0 0>, rot=<0 0 0 1>}
 R_nose_wrinkler_pose {type=Curve, loc=<0 0 0>, rot=<0 0 0 1>}
 base {type=Curve, loc=<0 0 0>, rot=<0 0 0 1>}
 eyeLookDownLeft {type=Curve, loc=<0 0 0>, rot=<0 0 0 1>}
 eyeLookInLeft {type=Curve, loc=<0 0 0>, rot=<0 0 0 1>}
 eyeLookUpLeft {type=Curve, loc=<0 0 0>, rot=<0 0 0 1>}
 eyeLookInRight {type=Curve, loc=<0 0 0>, rot=<0 0 0 1>}
 is_face_bone_anim {type=Curve, loc=<0 0 0>, rot=<0 0 0 1>}
 FullHourglass {type=Curve, loc=<0 0 0>, rot=<0 0 0 1>}
 LeanRectangle {type=Curve, loc=<0 0 0>, rot=<0 0 0 1>}
 InvertedTriangle {type=Curve, loc=<0 0 0>, rot=<0 0 0 1>}
 L_eye {type=Curve, loc=<0 0 0>, rot=<0 0 0 1>}
 R_eye {type=Curve, loc=<0 0 0>, rot=<0 0 0 1>}
 head {type=Curve, loc=<0 0 0>, rot=<0 0 0 1>}
 dyn_ear_l_master {type=Curve, loc=<0 0 0>, rot=<0 0 0 1>}
 root {type=Curve, loc=<0 0 0>, rot=<0 0 0 1>}
 dyn_ear02_l {type=Curve, loc=<0 0 0>, rot=<0 0 0 1>}
 dyn_ear03_l {type=Curve, loc=<0 0 0>, rot=<0 0 0 1>}
 dyn_ear04_l {type=Curve, loc=<0 0 0>, rot=<0 0 0 1>}
 dyn_ear05_l {type=Curve, loc=<0 0 0>, rot=<0 0 0 1>}
 dyn_ear01_l {type=Curve, loc=<0 0 0>, rot=<0 0 0 1>}
 flipbook_face_L_eye_index {type=Curve, loc=<0 0 0>, rot=<0 0 0 1>}
 flipbook_face_R_eye_index {type=Curve, loc=<0 0 0>, rot=<0 0 0 1>}
 flipbook_face_L_brow_index {type=Curve, loc=<0 0 0>, rot=<0 0 0 1>}
 flipbook_face_R_brow_index {type=Curve, loc=<0 0 0>, rot=<0 0 0 1>}
 flipbook_face_mouth_index {type=Curve, loc=<0 0 0>, rot=<0 0 0 1>}
 flipbook_face_bypass_mapping {type=Curve, loc=<0 0 0>, rot=<0 0 0 1>}
 flipbook_face_is_idling {type=Curve, loc=<0 0 0>, rot=<0 0 0 1>}
 root {loc=<0 0 0>, rot=<0 0 0 1>}
 pelvis {parent=root, loc=<0 -0.7321 92.4218>, rot=<0.0127 0.707 -0.0127 -0.707>}
 spine_01 {parent=pelvis, loc=<5.6002 -0.5128 0>, rot=<0 0 0.034 -0.9994>}
 spine_02 {parent=spine_01, loc=<8.0756 0.0139 0>, rot=<0 0 0.0238 -0.9997>}
 spine_03 {parent=spine_02, loc=<7.0395 0 0>, rot=<0 0 0.0149 -0.9999>}
 spine_04 {parent=spine_03, loc=<8.1023 0 0>, rot=<0 0 -0.0132 -0.9999>}
 spine_05 {parent=spine_04, loc=<10.0141 0 0>, rot=<0 0 -0.0015 -1>}
 clavicle_l {parent=spine_05, loc=<2.5856 -0.2866 -4.3777>, rot=<0.0465 -0.7289 -0.009 -0.6829>}
 upperarm_l {parent=clavicle_l, loc=<18.2264 0.1342 -0.001>, rot=<0.0253 -0.2693 0.0013 -0.9627>}
 lowerarm_l {parent=upperarm_l, loc=<29.0071 0.1686 -1.239>, rot=<0.0391 -0.0099 -0.2453 -0.9686>}
 hand_l {parent=lowerarm_l, loc=<25.8607 0.222 1.0274>, rot=<-0.7193 -0.0988 0.1193 -0.6773>}
 thumb_01_l {parent=hand_l, loc=<1.3118 -0.3191 -2.459>, rot=<0.6546 -0.4042 0.0664 -0.6354>}
 thumb_02_l {parent=thumb_01_l, loc=<5.9336 -0.0448 0.0401>, rot=<-0.0037 0.0213 0.298 -0.9543>}
 thumb_03_l {parent=thumb_02_l, loc=<2.2223 0.3682 -0.0907>, rot=<0.0118 -0.013 0.0066 -0.9998>}
 index_metacarpal_l {parent=hand_l, loc=<1.6535 0.0183 -2.96>, rot=<0.0158 -0.0124 -0.0173 -0.9996>}
 index_01_l {parent=index_metacarpal_l, loc=<8.3008 0.0357 -0.0524>, rot=<-0.0044 0.0485 0.0924 -0.9945>}
 index_02_l {parent=index_01_l, loc=<3.3587 -0.3004 0.0452>, rot=<-0.002 0.0027 0.3225 -0.9466>}
 index_03_l {parent=index_02_l, loc=<2.5578 -0.123 -0.0252>, rot=<-0.002 0.0067 0.2558 -0.9667>}
 middle_metacarpal_l {parent=hand_l, loc=<2.1856 -0.1725 -0.5984>, rot=<0.006 0.0168 -0.0329 -0.9993>}
 middle_01_l {parent=middle_metacarpal_l, loc=<7.3275 -0.2447 0.0129>, rot=<-0.0208 0.0042 0.0858 -0.9961>}
 middle_02_l {parent=middle_01_l, loc=<4.0123 0.1401 0.1175>, rot=<-0.0218 0.0258 0.3512 -0.9357>}
 middle_03_l {parent=middle_02_l, loc=<2.6077 0.0534 0.1373>, rot=<0.0082 0.0188 0.1708 -0.9851>}
 ring_metacarpal_l {parent=hand_l, loc=<2.2074 -0.3625 1.3638>, rot=<-0.0284 0.0722 -0.0458 -0.9959>}
 ring_01_l {parent=ring_metacarpal_l, loc=<7.0295 -0.6197 -0.1446>, rot=<-0.0489 -0.0252 0.1892 -0.9804>}
 ring_02_l {parent=ring_01_l, loc=<3.6174 0.4647 -0.1027>, rot=<0 0 0.2806 -0.9598>}
 ring_03_l {parent=ring_02_l, loc=<2.4152 0.114 0.0199>, rot=<0 0 0.1189 -0.9929>}
 pinky_metacarpal_l {parent=hand_l, loc=<1.7616 -0.9753 3.3814>, rot=<-0.0397 0.0814 -0.069 -0.9935>}
 pinky_01_l {parent=pinky_metacarpal_l, loc=<7.1709 -0.4831 -0.1094>, rot=<-0.0654 -0.0035 0.2242 -0.9723>}
 pinky_02_l {parent=pinky_01_l, loc=<2.8274 0.0893 -0.1043>, rot=<0.0037 0.0011 0.2854 -0.9584>}
 pinky_03_l {parent=pinky_02_l, loc=<2.2046 0.0307 -0.0078>, rot=<-0.0002 0.0098 0.1445 -0.9895>}
 weapon_l {parent=hand_l, loc=<-0.1974 -3.4111 0.3811>, rot=<0 0 0 1>}
 hand_attach_l {parent=hand_l, loc=<8.5 -3 0>, rot=<0.7071 0.7071 0 0>}
 deform_wrist_dn_base_l {parent=hand_l, loc=<-0.0022 -3.4251 -0.0258>, rot=<0.0025 -0.0018 0.0003 -1>}
 deform_wrist_dn_l {parent=deform_wrist_dn_base_l, loc=<0 0 0>, rot=<0 0 0 1>}
 deform_wrist_up_base_l {parent=hand_l, loc=<0.0019 2.7725 -0.0568>, rot=<0.0025 -0.0018 0.0003 -1>}
 deform_wrist_up_l {parent=deform_wrist_up_base_l, loc=<0 0 0>, rot=<0 0 0 1>}
 lowerarm_twist_02_l {parent=lowerarm_l, loc=<9.6199 0 0>, rot=<0 0 0 -1>}
 lowerarm_twist_01_l {parent=lowerarm_l, loc=<17.8549 0 0>, rot=<0.0031 -0.0004 -0.0041 -1>}
 dyn_L_sleeveScale {parent=lowerarm_twist_01_l, loc=<3.0701 0.9138 -0.2072>, rot=<0 0 0 1>}
 deform_elbow_in_base_l {parent=lowerarm_l, loc=<0.2112 5.645 0.8274>, rot=<0 0 0 1>}
 deform_elbow_in_l {parent=deform_elbow_in_base_l, loc=<0 0 0>, rot=<0 0 0 1>}
 deform_elbow_out_base_l {parent=lowerarm_l, loc=<0 -7.5129 0>, rot=<0 0 0 1>}
 deform_elbow_out_l {parent=deform_elbow_out_base_l, loc=<0 0 0>, rot=<0 0 0 1>}
 crv_lowerarm_l_bottom {parent=lowerarm_l, loc=<0.1879 0.4312 0.0442>, rot=<0 0 0 1>}
 crv_lowerarm_l_top {parent=lowerarm_l, loc=<-0.0117 0.1651 -0.0027>, rot=<0 0 0 1>}
 upperarm_twist_01_l {parent=upperarm_l, loc=<7.9796 0.0843 -0.6195>, rot=<0 0 0 -1>}
 deform_upperarm_bk_base_l {parent=upperarm_twist_01_l, loc=<-6.8983 -13.3826 -1.3258>, rot=<0 0 0 1>}
 deform_upperarm_bk_l {parent=deform_upperarm_bk_base_l, loc=<0 0 0>, rot=<0 0 0 1>}
 deform_upperarm_fr_base_l {parent=upperarm_twist_01_l, loc=<-8.3487 10.7117 -2.1455>, rot=<0 0 0 1>}
 deform_upperarm_fr_l {parent=deform_upperarm_fr_base_l, loc=<0 0 0>, rot=<0 0 0 1>}
 deform_upperarm_up_base_l {parent=upperarm_twist_01_l, loc=<-11.2012 -1.2234 8.7784>, rot=<0 0 0 1>}
 deform_upperarm_up_l {parent=deform_upperarm_up_base_l, loc=<0 0 0>, rot=<0 0 0 1>}
 upperarm_twist_02_l {parent=upperarm_l, loc=<19.5059 0.0843 -0.6195>, rot=<0 0 0 -1>}
 aim_upperarm_l {parent=upperarm_l, loc=<5.1736 0 1.3993>, rot=<0 0 0 -1>}
 neck_01 {parent=spine_05, loc=<10.815 1.9196 0>, rot=<0 0 -0.1864 -0.9825>}
 neck_02 {parent=neck_01, loc=<4.7347 0.3045 0>, rot=<0 0 0.0363 -0.9993>}
 head {parent=neck_02, loc=<5.7632 0.5552 0>, rot=<0 0 0.115 -0.9934>}
 clavicle_r {parent=spine_05, loc=<2.5855 -0.2866 4.3777>, rot=<0.7289 0.0465 -0.6829 0.009>}
 upperarm_r {parent=clavicle_r, loc=<-18.2264 -0.1342 0.001>, rot=<0.0253 -0.2693 0.0013 -0.9627>}
 lowerarm_r {parent=upperarm_r, loc=<-29.0071 -0.1686 1.239>, rot=<0.0391 -0.0099 -0.2453 -0.9686>}
 hand_r {parent=lowerarm_r, loc=<-25.8607 -0.222 -1.0274>, rot=<-0.7193 -0.0988 0.1193 -0.6773>}
 thumb_01_r {parent=hand_r, loc=<-1.3118 0.3191 2.459>, rot=<0.6546 -0.4042 0.0664 -0.6354>}
 thumb_02_r {parent=thumb_01_r, loc=<-5.9336 0.0448 -0.0401>, rot=<-0.0037 0.0213 0.298 -0.9543>}
 thumb_03_r {parent=thumb_02_r, loc=<-2.2223 -0.3682 0.0907>, rot=<0.0118 -0.013 0.0066 -0.9998>}
 index_metacarpal_r {parent=hand_r, loc=<-1.6535 -0.0183 2.96>, rot=<0.0158 -0.0124 -0.0173 -0.9996>}
 index_01_r {parent=index_metacarpal_r, loc=<-8.3008 -0.0357 0.0524>, rot=<-0.0044 0.0485 0.0924 -0.9945>}
 index_02_r {parent=index_01_r, loc=<-3.3587 0.3004 -0.0452>, rot=<-0.002 0.0027 0.3225 -0.9466>}
 index_03_r {parent=index_02_r, loc=<-2.5578 0.123 0.0252>, rot=<-0.002 0.0067 0.2558 -0.9667>}
 middle_metacarpal_r {parent=hand_r, loc=<-2.1856 0.1725 0.5984>, rot=<0.006 0.0168 -0.0329 -0.9993>}
 middle_01_r {parent=middle_metacarpal_r, loc=<-7.3275 0.2447 -0.0129>, rot=<-0.0208 0.0042 0.0858 -0.9961>}
 middle_02_r {parent=middle_01_r, loc=<-4.0123 -0.1401 -0.1175>, rot=<-0.0218 0.0258 0.3512 -0.9357>}
 middle_03_r {parent=middle_02_r, loc=<-2.6077 -0.0534 -0.1373>, rot=<0.0082 0.0188 0.1708 -0.9851>}
 ring_metacarpal_r {parent=hand_r, loc=<-2.2074 0.3625 -1.3638>, rot=<-0.0284 0.0722 -0.0458 -0.9959>}
 ring_01_r {parent=ring_metacarpal_r, loc=<-7.0295 0.6197 0.1446>, rot=<-0.0489 -0.0252 0.1892 -0.9804>}
 ring_02_r {parent=ring_01_r, loc=<-3.6174 -0.4647 0.1027>, rot=<0 0 0.2806 -0.9598>}
 ring_03_r {parent=ring_02_r, loc=<-2.4152 -0.114 -0.0199>, rot=<0 0 0.1189 -0.9929>}
 pinky_metacarpal_r {parent=hand_r, loc=<-1.7616 0.9753 -3.3814>, rot=<-0.0397 0.0814 -0.069 -0.9935>}
 pinky_01_r {parent=pinky_metacarpal_r, loc=<-7.1709 0.4831 0.1094>, rot=<-0.0654 -0.0035 0.2242 -0.9723>}
 pinky_02_r {parent=pinky_01_r, loc=<-2.8274 -0.0893 0.1043>, rot=<0.0037 0.0011 0.2854 -0.9584>}
 pinky_03_r {parent=pinky_02_r, loc=<-2.2046 -0.0307 0.0078>, rot=<-0.0002 0.0098 0.1445 -0.9895>}
 hand_attach_r {parent=hand_r, loc=<-8.5 3 0>, rot=<0 0 -0.7071 -0.7071>}
 weapon_r {parent=hand_r, loc=<0.1974 3.4111 -0.3811>, rot=<0 0 0 1>}
 deform_wrist_dn_base_r {parent=hand_r, loc=<0 3.4172 0>, rot=<0 0 0 -1>}
 deform_wrist_dn_r {parent=deform_wrist_dn_base_r, loc=<0 0 0>, rot=<0 0 0 -1>}
 deform_wrist_up_base_r {parent=hand_r, loc=<-0.0001 -2.7804 0>, rot=<0 0 0 -1>}
 deform_wrist_up_r {parent=deform_wrist_up_base_r, loc=<0 0 0>, rot=<0 0 0 -1>}
 lowerarm_twist_02_r {parent=lowerarm_r, loc=<-9.6199 0 0>, rot=<0 0 0 -1>}
 lowerarm_twist_01_r {parent=lowerarm_r, loc=<-17.8549 0 0>, rot=<0 0 0 -1>}
 dyn_R_sleeveScale {parent=lowerarm_twist_01_r, loc=<-3.0621 -0.9378 0.2153>, rot=<0.0031 -0.0004 -0.0041 -1>}
 deform_elbow_in_base_r {parent=lowerarm_r, loc=<-0.2114 -5.6449 -0.827>, rot=<0 0 0 -1>}
 deform_elbow_in_r {parent=deform_elbow_in_base_r, loc=<0 0 0>, rot=<0 0 0 1>}
 deform_elbow_out_base_r {parent=lowerarm_r, loc=<-0.0001 7.5129 0.0002>, rot=<0 0 0 -1>}
 deform_elbow_out_r {parent=deform_elbow_out_base_r, loc=<0 0 0>, rot=<0 0 0 1>}
 crv_lowerarm_r_top {parent=lowerarm_r, loc=<0.0117 0.215 0.0027>, rot=<0 0 0 1>}
 crv_lowerarm_r_bottom {parent=lowerarm_r, loc=<-0.1879 -0.3704 -0.0442>, rot=<0 0 0 1>}
 upperarm_twist_01_r {parent=upperarm_r, loc=<-7.9796 -0.0843 0.6195>, rot=<0 0 0 -1>}
 deform_upperarm_bk_base_r {parent=upperarm_twist_01_r, loc=<6.8985 13.3826 1.3257>, rot=<0 0 0 1>}
 deform_upperarm_bk_r {parent=deform_upperarm_bk_base_r, loc=<0 0 0>, rot=<0 0 0 1>}
 deform_upperarm_fr_base_r {parent=upperarm_twist_01_r, loc=<8.3485 -10.7117 2.1456>, rot=<0 0 0 1>}
 deform_upperarm_fr_r {parent=deform_upperarm_fr_base_r, loc=<0 0 0>, rot=<0 0 0 1>}
 deform_upperarm_up_base_r {parent=upperarm_twist_01_r, loc=<11.2015 1.2234 -8.7788>, rot=<0 0 0 1>}
 deform_upperarm_up_r {parent=deform_upperarm_up_base_r, loc=<0 0 0>, rot=<0 0 0 1>}
 upperarm_twist_02_r {parent=upperarm_r, loc=<-19.5059 -0.0843 0.6195>, rot=<0 0 0 -1>}
 aim_upperarm_r {parent=upperarm_r, loc=<-5.174 0 -1.399>, rot=<0 0 0 1>}
 attach_cape {parent=spine_05, loc=<9.017 -11.714 0>, rot=<0.0616 -0.7044 -0.0616 -0.7044>}
 cape {parent=attach_cape, type=Null, loc=<0 -6.9086 0>, rot=<0 0 0 1>}
 attach_backpack {parent=spine_05, loc=<-0.123 -13.473 0>, rot=<-0.0093 -0.707 0.0093 -0.707>}
 Backpack_BR {parent=attach_backpack, type=Null, loc=<0 -8.6755 0>, rot=<0 0 0 1>}
 Beared_Mover_R {parent=spine_05, loc=<-2.455 19.397 7.878>, rot=<0.0284 -0.7065 -0.0284 -0.7065>}
 Beared_Mover_L {parent=spine_05, loc=<-2.4553 19.3974 -7.8781>, rot=<0.0284 -0.7065 -0.0284 -0.7065>}
 deform_pec_aim_base_l {parent=spine_04, loc=<5.0227 13.6719 -2.6508>, rot=<0.0672 -0.5201 0.3228 -0.7879>}
 deform_pec_aim_l {parent=deform_pec_aim_base_l, loc=<0 0 0>, rot=<0 0 0 1>}
 deform_pec_base_l {parent=deform_pec_aim_l, loc=<16.3279 0 0>, rot=<0 0 0 1>}
 deform_pec_l {parent=deform_pec_base_l, loc=<0 0 0>, rot=<0 0 0 1>}
 deform_pec_aim_base_r {parent=spine_04, loc=<5.0225 13.6718 2.6508>, rot=<-0.5201 -0.0672 0.7879 0.3228>}
 deform_pec_aim_r {parent=deform_pec_aim_base_r, loc=<0 0 0>, rot=<0 0 0 1>}
 deform_pec_base_r {parent=deform_pec_aim_r, loc=<-16.3279 0 0>, rot=<0 0 0 -1>}
 deform_pec_r {parent=deform_pec_base_r, loc=<0 0 0.0002>, rot=<0 0 0 -1>}
 Chest_mover_l {parent=spine_04, loc=<-2.527 2.657 -21.185>, rot=<0.038 -0.9137 -0.4043 0.0168>}
 Chest_mover_r {parent=spine_04, loc=<-2.527 2.657 21.185>, rot=<0.038 -0.9137 -0.4043 0.0168>}
 thigh_l {parent=pelvis, loc=<-1.8716 -0.0109 -11.9936>, rot=<0.0012 0.0572 0.0155 -0.9982>}
 calf_l {parent=thigh_l, loc=<-41.9178 -0.2246 0.0926>, rot=<0.0006 -0.0161 -0.0391 -0.9991>}
 foot_l {parent=calf_l, loc=<-38.7057 1.2438 0.2452>, rot=<0.0975 -0.0474 0.0317 -0.9936>}
 ball_l {parent=foot_l, loc=<-6.1614 13.1469 0.6807>, rot=<0.0001 -0.0001 -0.7063 -0.7079>}
 Base_Shoelace_L {parent=foot_l, loc=<-2.0586 8.6198 1.4811>, rot=<-0.3971 -0.5969 0.309 -0.6249>}
 dyn_Shoelace_L1 {parent=Base_Shoelace_L, loc=<0 0 0>, rot=<0 0 0 1>}
 dyn_Shoelace_L2 {parent=Base_Shoelace_L, loc=<0 0 0>, rot=<0 -1 0 0>}
 calf_twist_02_l {parent=calf_l, loc=<-15.9268 -0.0967 0.0397>, rot=<0.0237 0.0095 -0.0035 -0.9997>}
 calf_twist_01_l {parent=calf_l, loc=<-32.9334 -0.0967 0.0397>, rot=<0.0922 0.0098 -0.0029 -0.9957>}
 deform_knee_in_base_l {parent=calf_l, loc=<1.298 -8.0356 -0.0496>, rot=<0 0 0 1>}
 deform_knee_in_l {parent=deform_knee_in_base_l, loc=<0 0 0>, rot=<0 0 0 1>}
 deform_knee_out_base_l {parent=calf_l, loc=<0 6.8739 0>, rot=<0 0 0 -1>}
 deform_knee_out_l {parent=deform_knee_out_base_l, loc=<0 0 0>, rot=<0 0 0 1>}
 crv_calf_l_bottom {parent=calf_l, loc=<-0.0053 0 0>, rot=<0 0 0 1>}
 crv_calf_l_top {parent=calf_l, loc=<0 0 0>, rot=<0 0 0 1>}
 thigh_twist_01_l {parent=thigh_l, loc=<-21.6303 0 0>, rot=<0.0013 0 0 -1>}
 deform_groin_base_l {parent=thigh_l, loc=<-2.1908 8.5661 -1.7442>, rot=<0 0 0 -1>}
 deform_groin_l {parent=deform_groin_base_l, loc=<0 0 0>, rot=<0 0 0 1>}
 thigh_r {parent=pelvis, loc=<-1.8716 -0.0109 11.9936>, rot=<0.0572 -0.0012 0.9982 0.0155>}
 calf_r {parent=thigh_r, loc=<41.9178 0.2246 -0.0926>, rot=<0.0006 -0.0161 -0.0391 -0.9991>}
 foot_r {parent=calf_r, loc=<38.7057 -1.2438 -0.2452>, rot=<0.0975 -0.0474 0.0317 -0.9936>}
 ball_r {parent=foot_r, loc=<6.1614 -13.1469 -0.6807>, rot=<0.0001 -0.0001 -0.7063 -0.7079>}
 Base_Shoelace_R {parent=foot_r, loc=<2.0586 -8.6198 -1.4812>, rot=<0.6584 -0.2698 -0.6248 -0.3214>}
 dyn_Shoelace_R1 {parent=Base_Shoelace_R, loc=<0 0 0>, rot=<0 0 0 1>}
 dyn_Shoelace_R2 {parent=Base_Shoelace_R, loc=<0 0 0>, rot=<0 -1 0 0>}
 calf_twist_02_r {parent=calf_r, loc=<15.9268 0.0967 -0.0397>, rot=<0.0237 0.0095 -0.0035 -0.9997>}
 calf_twist_01_r {parent=calf_r, loc=<32.9336 0.0967 -0.0397>, rot=<0.1193 0.0098 -0.0026 -0.9928>}
 deform_knee_in_base_r {parent=calf_r, loc=<-1.2981 8.0356 0.0496>, rot=<0 0 0 1>}
 deform_knee_in_r {parent=deform_knee_in_base_r, loc=<0 0 0>, rot=<0 0 0 1>}
 deform_knee_out_base_r {parent=calf_r, loc=<0 -6.8739 0>, rot=<0 0 0 -1>}
 deform_knee_out_r {parent=deform_knee_out_base_r, loc=<0 0 0>, rot=<0 0 0 1>}
 crv_calf_r_bottom {parent=calf_r, loc=<0 0 0>, rot=<0 0 0 1>}
 crv_calf_r_top {parent=calf_r, loc=<0 0 0>, rot=<0 0 0 1>}
 thigh_twist_01_r {parent=thigh_r, loc=<21.6303 0 0>, rot=<0.0013 0 0 -1>}
 deform_groin_base_r {parent=thigh_r, loc=<2.1908 -8.5661 1.7442>, rot=<0 0 0 -1>}
 deform_groin_r {parent=deform_groin_base_r, loc=<0 0 0>, rot=<0 0 0 -1>}
 deform_glute_base_l {parent=pelvis, loc=<-10.1877 -11.6217 -12.899>, rot=<0.0275 0.0501 -0.4472 -0.8926>}
 deform_glute_l {parent=deform_glute_base_l, loc=<0 0 0>, rot=<0 0 0 1>}
 deform_glute_base_r {parent=pelvis, loc=<-10.1877 -11.6218 12.899>, rot=<0.0501 -0.0275 0.8926 -0.4472>}
 deform_glute_r {parent=deform_glute_base_r, loc=<0 0 0>, rot=<0 0 0 1>}
 attach {parent=root, loc=<0 0 0>, rot=<0 0 0 1>}
 ik_hand_root {parent=root, loc=<0 0 0>, rot=<0 0 0 1>}
 ik_hand_gun {parent=ik_hand_root, loc=<-64.462 4.8638 101.1961>, rot=<0.6084 0.1793 0.4613 -0.6205>}
 ik_hand_r {parent=ik_hand_gun, loc=<0 0 0>, rot=<0 0 0 1>}
 ik_hand_l {parent=ik_hand_gun, loc=<65.7708 101.9258 43.6659>, rot=<-0.5102 -0.7906 -0.3387 0>}
 ik_foot_root {parent=root, loc=<0 0 0>, rot=<0 0 0 1>}
 ik_foot_l {parent=ik_foot_root, loc=<19.6189 -3.2051 10.3339>, rot=<0.0753 0.7018 0.0677 -0.7051>}
 ik_foot_r {parent=ik_foot_root, loc=<-19.6189 -3.2051 10.3339>, rot=<0.7051 0.0677 -0.7018 0.0753>}
 root_fp {parent=root, loc=<0 0 0>, rot=<0 0 0 1>}
 camera_root {parent=root_fp, loc=<0 0 0>, rot=<0 0 0 1>}
 camera {parent=camera_root, loc=<0 0 0>, rot=<0 0 0 1>}
 attach_fp {parent=root_fp, loc=<0 0 0>, rot=<0 0 0 1>}
 attach_emote_01 {parent=root, loc=<0 0 0>, rot=<0 0 0 1>}
 attach_emote_02 {parent=root, loc=<0 0 0>, rot=<0 0 0 1>}
 attach_emote_03 {parent=root, loc=<0 0 0>, rot=<0 0 0 1>}
 warp_point_01 {parent=root, loc=<0 0 0>, rot=<0 0 0 1>}
 warp_point_02 {parent=root, loc=<0 0 0>, rot=<0 0 0 1>}
 warp_point_03 {parent=root, loc=<0 0 0>, rot=<0 0 0 1>}
]
variables [is holding gun {type=double, default=0}]
